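# Material function that produces cheap 1D noise in the 0 to 1 range from a scalar input, which defaults to an offset UV X coordinate.
# It evaluates four cosine waves at once in a float4, using exposed frequency coefficients and fixed phase offsets.
# A dot product with exposed wave weights sums the waves before a bias-scale remaps the result.
# Debug graph plotting and other leftover nodes do not feed the output.
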MaterialFunction /Game/Environments/Asteria/Sets/NeonCity/Techart/Screen/MaterialFunctions/MF_Cheap1DNoise
graph {nodes [
  n1 = Subtract(A=ComponentMask(None=TextureCoordinate()) {r=true}) {constB=0.1512}
  n2 = FunctionInput(Preview=@n1) {bUsePreviewValueAsDefault=true, inputType=FunctionInput_Scalar}
  n3 = FunctionInput(Preview=Constant4Vector() {constant=<rgba 0.8954 2.3215 4.1247 8.6325>}.RGBA) {bUsePreviewValueAsDefault=true, inputName=Frequency Coefficients, inputType=FunctionInput_Vector4}
  n4 = Cosine(None=Add(A=Multiply(A=@n2, B=@n3), B=Constant4Vector() {constant=<rgba 0.0322 0.3542 0.55 0.8754>}.RGBA))
  Add(A=@n4)
  Constant4Vector() {constant=<rgba 1 2 4 8>}
  n5 = Constant4Vector() {constant=<rgba 1 1 1 1>}
  ConstantBiasScale(None=Divide(A=DotProduct(B=@n5.RGBA)) {constB=4})
  Constant()
  Constant() {r=1}
  n6 = FunctionInput(Preview=Constant4Vector() {constant=<rgba 0.25 0.25 0.25 0.25>}.RGBA) {bUsePreviewValueAsDefault=true, description="R is freq 1, G is freq 2 etc", inputName=Wave Weights, inputType=FunctionInput_Vector4}
  n7 = DotProduct(A=@n4, B=@n6)
  Divide(A=@n7) {constB=4}
  FunctionOutput(None=ConstantBiasScale(None=@n7)) {description=0 to 1, outputName=Noise_0_1}
  n8 = StaticBool()
  n9 = Constant() {r=2}
  MaterialFunctionCall!(X range Y range, Gradient Height, ShowGrid?=@n8, Thickness=@n9, Color, "f(x)") {materialFunction=/Engine/Functions/Engine_MaterialFunctions01/Debug/PlotFunctionOnGraph}
  MaterialFunctionCall(ShowGrid?=@n8, Thickness=@n9, "f(x)"=@n7) {materialFunction=&PlotFunctionOnGraph}
  Multiply(A=Time()) {constB=0.1}
  Saturate(None=Sign())
  SphereMask(B=Constant()) {attenuationRadius=1, hardnessPercent=0}
  Subtract(B=Multiply(A=@n1) {constB=5})
 ]}
