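# Material function that balances emissive brightness for NeonCity screen and sign materials in the Asteria environment.
# Expands the LDR emissive input to HDR via DeriveHDRfromLDR, applies an instance multiplier and eye-adaptation function, then scales it by time of day.
# The ray-traced branch adds Lumen color override, boost and desaturation, separate from the normal path.
# Low quality instead multiplies the balanced emissive by a low-end lighting scalar.
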MaterialFunction /Game/Environments/Asteria/Sets/NeonCity/Techart/Screen/MaterialFunctions/MF_NeonCityEmissiveBalance
graph {nodes [
  n1 = FunctionInput() {inputName=Emissive}
  n2 = ScalarParameter() {defaultValue=50, group=Global Emissive Adjustments, parameterName=SignDynamicRange}
  n3 = Multiply(
   A=ScalarParameter() {defaultValue=4, group=Global Emissive Adjustments, parameterName=GlobalBrightnessTweak}
   B=ScalarParameter() {defaultValue=10, parameterName=EmTweak}
  )
  n4 = ScalarParameter() {defaultValue=1, group=Global Emissive Adjustments, parameterName=HDR Power}
  n5 = MaterialFunctionCall!(
   Emissive=Multiply(
    A=MaterialFunctionCall(LDR Input=@n1, HDR Dynamic Range=@n2, HDR Intensity Multiplier=@n3, Derive HDR Power=@n4) {materialFunction=&DeriveHDRfromLDR}.RemappedRange
    B=ScalarParameter() {defaultValue=1, group=Global Emissive Adjustments, parameterName=Instance Emissive Multiplier}
   )
   Alpha=ScalarParameter() {defaultValue=1, group=Global Emissive Adjustments, parameterName=Cosmetic Emissive EyeAdaptation Alpha}
   Material
  ) {materialFunction=/Game/Packages/Fortress_SharedMaterials/Base_Material_Functions/BaseCharacterShaderFunctions/MF_CosmeticEmissiveEyeAdaptation}
  n6 = Multiply(
   A=@n5.Emissive
   B=MaterialFunctionCall(
    Value_Morning=ScalarParameter() {defaultValue=0.65, group=Global Emissive Adjustments, parameterName=Emissive_Morning}
    Value_Day=ScalarParameter() {defaultValue=0.6, group=Global Emissive Adjustments, parameterName=Emissive_Day}
    Value_Evening=ScalarParameter() {defaultValue=0.65, group=Global Emissive Adjustments, parameterName=Emissive_Evening}
    Value_Night=Constant() {r=1}
   ) {materialFunction=&Scalar_Time_of_Day}.Result
  )
  n7 = VectorParameter() {parameterName=LumenColorOverride, defaultValue=<rgba 1 1 1 1>}
  n8 = ScalarParameter() {parameterName=LumenUseColorOverride}
  n9 = Desaturation(
   None=Multiply(A=ScalarParameter() {defaultValue=2, parameterName=LumenBoost}, B=LinearInterpolate(A=@n6, B=@n7.RGB, Alpha=@n8))
   Fraction=ScalarParameter() {parameterName=LumenSaturation}
  )
  n10 = ScalarParameter() {defaultValue=0.5, parameterName=LowEndLightingMultiplier}
  FunctionOutput(None=QualitySwitch(Default=RayTracingQualitySwitch(Normal=@n6, RayTraced=@n9), Low=Multiply(A=@n6, B=@n10)))
 ]}
