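# MF_Rand1Dto1D1 is a material function in the NeonCity screen techart set that turns one scalar input into a pseudo-random scalar.
# It uses the classic sine hash: the input is scaled by a constant, passed through Sine, multiplied by a large constant, then reduced with Frac.
# The result exits as Random1D, and a constant feeds the input's preview value.
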MaterialFunction /Game/Environments/Asteria/Sets/NeonCity/Techart/Screen/MaterialFunctions/MF_Rand1Dto1D1
graph {nodes [
  n1 = Constant() {r=0.3333}
  n2 = Sine(None=Multiply(A=FunctionInput(Preview=@n1) {inputType=FunctionInput_Scalar}, B=Constant() {r=12.9898}))
  n3 = Constant() {r=43758.5469}
  FunctionOutput(None=Frac(None=Multiply(A=@n2, B=@n3))) {outputName=Random1D}
 ]}
